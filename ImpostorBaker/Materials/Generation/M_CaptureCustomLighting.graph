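# M_CaptureCustomLighting is a translucent deferred decal material in ImpostorBaker's generation set, apparently for custom-lighting capture, judging by its name.
# Its graph holds only two scalar constants, one and zero, with no links to material outputs shown.
# It uses the Default Lit shading model and has no textures, parameters or custom lighting math.
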Material /ImpostorBaker/Materials/Generation/M_CaptureCustomLighting
graph {
 properties {blendMode=BLEND_Translucent, materialDomain=MD_DeferredDecal, pixelDepthOffsetMode=PDOM_Legacy, usedShadingModels=Default Lit}
 nodes [
  Constant() {r=1}
  Constant()
 ]
}
