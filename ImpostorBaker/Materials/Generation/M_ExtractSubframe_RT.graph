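# Unlit, alpha-composite material in the Impostor Baker's generation set, built to extract one subframe from a render target laid out as a 3x3 grid.
# The FrameIndex parameter becomes a 2D cell index through 1Dto2DIndex, which offsets one-third-scale UVs into that cell.
# Those UVs sample the RT texture parameter, which defaults to black.
# A separate sample of a test render target has no connections shown.
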Material /ImpostorBaker/Materials/Generation/M_ExtractSubframe_RT
graph {
 properties {blendMode=BLEND_AlphaComposite, pixelDepthOffsetMode=PDOM_Legacy, shadingModel=MSM_Unlit, usedShadingModels=Unlit}
 nodes [
  n1 = TextureCoordinate()
  n2 = MaterialFunctionCall!(1D Index=ScalarParameter() {parameterName=FrameIndex}, Cells XY=Constant() {r=3}) {materialFunction=/Engine/Functions/Engine_MaterialFunctions01/Coordinates/1Dto2DIndex}
  TextureSampleParameter2D(UVs=Add(A=Divide(A=@n1) {constB=3}, B=Divide(A=@n2.2D Index) {constB=3})) {parameterName=RT, texture=&Black}
  TextureSample() {texture=/ImpostorBaker/TestContent/RTs/RT_Test_01}
 ]
}
