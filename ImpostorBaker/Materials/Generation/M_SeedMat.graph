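# Unlit material in the Impostor Baker's generation set, named as a seed material, that computes repeating UVs for a grid of frames.
# It multiplies the texture coordinates by the XYFrames scalar parameter and takes the fractional part, giving each grid cell its own 0–1 UV range.
# The graph samples no textures, and no link from the result to a material output is shown.
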Material /ImpostorBaker/Materials/Generation/M_SeedMat
graph {
 properties {pixelDepthOffsetMode=PDOM_Legacy, shadingModel=MSM_Unlit, usedShadingModels=Unlit}
 nodes [Frac(None=Multiply(A=TextureCoordinate(), B=ScalarParameter() {defaultValue=12, parameterName=XYFrames}))]
}
