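# Unlit Impostor Baker material that computes corner cuts for an impostor cutout mesh from a mask render target, one per UV quadrant.
# A Custom node scans a 16x16 texel grid inward from each corner for two anchor points.
# A second finds steepest non-clipping lines between them, keeping the largest-area cut as two normalized vertices packed in a float4.
# Parameters toggle debug views of anchors, candidate lines and numeric values.
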Material /ImpostorBaker/Materials/Generation/M_TraceMeshCutout
graph {
 properties {blendMode=BLEND_AlphaComposite, pixelDepthOffsetMode=PDOM_Legacy, shadingModel=MSM_Unlit, usedShadingModels=Unlit}
 nodes [
  n1 = NamedRerouteUsage()
  LinearInterpolate(A=OneMinus(None=ComponentMask(None=@n1) {a=true}), Alpha=ScalarParameter() {defaultValue=1, parameterName=Enable Debug})
  n2 = ComponentMask(None=NamedRerouteUsage()) {b=true, g=true, r=true}
  n3 = NamedRerouteUsage()
  n4 = NamedRerouteUsage()
  n5 = ScalarParameter() {defaultValue=1, parameterName=Debug Anchors}
  n6 = Add(A=Add(A=LinearInterpolate(A=@n3, B=@n4, Alpha=@n5), B=NamedRerouteUsage()), B=NamedRerouteUsage())
  n7 = ScalarParameter() {defaultValue=1, parameterName=Debug Lines}
  LinearInterpolate(
   A=ComponentMask(None=NamedRerouteUsage()) {b=true, g=true, r=true}
   B=LinearInterpolate(A=LinearInterpolate(A=@n2, B=@n6, Alpha=@n7), B=NamedRerouteUsage(), Alpha=ScalarParameter() {parameterName=Debug Values})
   Alpha=ScalarParameter() {defaultValue=1, parameterName=Enable Debug}
  )
  MaterialExpressionNamedRerouteUsage_7 = NamedRerouteUsage()
  MaterialExpressionReroute_0 = Reroute(None=NamedRerouteUsage())
  n8 = TextureCoordinate()
  MaterialExpressionNamedRerouteDeclaration_2 = NamedRerouteDeclaration(None=Floor(None=Multiply(A=@n8) {constB=2})) {name=Quadrant, nodeColor=<rgba 1 0 0.1406 1>}
  MaterialExpressionTextureCoordinate_4 = TextureCoordinate()
  MaterialExpressionScalarParameter_11 = ScalarParameter() {defaultValue=0.25, parameterName=Threshold}
  n9 = Custom!(
   Anchors=@MaterialExpressionNamedRerouteUsage_7
   Tex=@MaterialExpressionReroute_0
   Quadrant=@MaterialExpressionNamedRerouteDeclaration_2
   UV=@MaterialExpressionTextureCoordinate_4
   Threshold=@MaterialExpressionScalarParameter_11
  ) {
   code="struct ShapeDrawer\r\n{\r\n\r\n\tfloat2 P;\r\n\t\r\n\tfloat DrawLine(float2 inA, float2 inB, float inWidth)\r\n\t{\r\n\t\tfloat2 UV = P - inA;\r\n\t\tfloat2 squashvec = inB - inA;\r\n\t\t\r\n\t\t\r\n\t\t\r\n\t\tfloat2 distalongline =  saturate( dot(UV, squashvec) / dot(squashvec, squashvec)) * squashvec;\r\n\t\t\r\n\t\tfloat2 squashUV = UV - distalongline;\r\n\t\t\r\n\t\tfloat thisline = length(squashUV);\r\n\t\tthisline = saturate(-((thisline - inWidth)/ inWidth));\r\n\t\t\r\n\t\treturn thisline;\r\n\t}\r\n\t\r\n\tfloat DrawCircle(float2 inC, float inRadius, float inWidth)\r\n\t{\r\n\t\tfloat2 UV = P - inC;\r\n\t\tinWidth/=2;\r\n\t\r\n\t\tfloat dist = abs( length(UV) - inRadius);\r\n\t\t\r\n\t\tdist = saturate(-((dist - inWidth)/ inWidth));\r\n\t\treturn dist;\r\n\t}\r\n\t\r\n\tfloat DrawBox(float2 inC, float2 inExtents, float inWidth)\r\n\t{\r\n\t\tfloat2 pos = abs(P - inC) / inExtents;\r\n\t\tfloat dist = abs(max(pos.x-1, pos.y-1));\r\n\t\tdist = dist - inWidth*2;\r\n\t\tdist = saturate(-((dist - inWidth)/ inWidth));\r\n\t\t\r\n\t\treturn dist;\r\n\t}\r\n\t\r\n\tfloat2 GetFocalDirection(float2 inFocalPoint, float2 inRayAtFocalDepth, float2 inOffset)\r\n\t{\r\n\t\tfloat2 gd = normalize( inRayAtFocalDepth - inOffset);\r\n\t\treturn normalize(inRayAtFocalDepth + gd);\r\n\t}\r\n\t\r\n\tfloat2 IntersectCircle( float2 inC, float inR, float2 inRO, float2 inRD)\r\n\t{\r\n\t\tinRO = inRO - inC;\r\n\t\tfloat B = dot( inRD, inRO);\r\n\t\tfloat C = dot( inRO, inRO) - (inR*inR);\r\n\t\tfloat D = sqrt(B*B - C);\r\n\t\tfloat t0 = -B - D;\r\n\t\tfloat t1 = -B + D;\r\n\t\t\r\n\t\tt0 = max(0, t0);\r\n\t\t\t\t\r\n\t\treturn float2(t0, t1);\r\n\t}\r\n\t\r\n\tfloat2 IntersectBox(float2 inC, float2 inExtents, float2 inRO, float2 inRD)\r\n\t{\r\n\t\tinRO -= inC;\r\n\t\t\r\n\t\tfloat2 invraydir = 1.0f / inRD;\r\n\t\tfloat2 firstintersections = (-inExtents - inRO) * invraydir;\r\n\t\tfloat2 secondintersections = (inExtents - inRO) * invraydir;\r\n\t\tfloat2 closest = min(firstintersections, secondintersections);\r\n\t\tfloat2 furthest = max(firstintersections, secondintersections);\r\n\t\t\r\n\t\tfloat t0 = max(closest.x, closest.y);\r\n\t\tfloat t1 = min(furthest.x, furthest.y);\r\n\t\tt0 = max(0, t0);\r\n\t\t\r\n\t\treturn float2(t0, t1);\r\n\t}\r\n\t\r\n};\r\n\r\nShapeDrawer Helper;\r\nHelper.P = UV * 16;\r\n\r\nfloat3 debug_lines = 0;\r\n\r\n\r\nfloat texelsize = 1.0 / 16.0;\r\nfloat cursample = 0;\r\n\r\nfloat2 vertexA = Anchors.ba;\r\nfloat2 vertexB = Anchors.rg;\r\n\r\n//if the Bounds are 0, the result is just a square \r\nif((vertexA.x == vertexB.x &&&& vertexA.y == vertexB.y))\r\n{\r\n\tvertexA.x = Quadrant.x == 0.0 ? vertexA.x : 15.0 - vertexA.x;\r\n\tvertexA.y = Quadrant.y == 0.0 ? vertexA.y : 15.0 - vertexA.y;\r\n\t\t\r\n\tvertexB.x = Quadrant.x == 0.0 ? vertexB.x : 15.0 - vertexB.x;\r\n\tvertexB.y = Quadrant.y == 0.0 ? vertexB.y : 15.0 - vertexB.y;\r\n\r\n\tvertexA += saturate(Quadrant);\r\n\tvertexB += saturate(Quadrant);\r\n\r\n\treturn float4(vertexA, vertexB) / 16;\r\n}\r\n\r\nfloat maxslope = 0;\r\n\r\nfloat maxslopes[8] = {0,0,0,0,0,0,0,0};\r\nfloat4 maxslopetris[8] = {float4(0,0,0,0),float4(0,0,0,0),float4(0,0,0,0),float4(0,0,0,0),float4(0,0,0,0),float4(0,0,0,0),float4(0,0,0,0),float4(0,0,0,0)};\r\n\r\n//bounds in the form of XminXmax, YminYmax\r\nfloat4 bounds = float4( vertexA.x, vertexB.x, vertexB.y, vertexA.y);\r\n\r\nfloat2 curUV = float2(0,0);\r\n\r\n//trace down from the corner within the bounds between vertices\r\nfor(int i = bounds.x + 1; i <= bounds.y; i++)\r\n{\r\n\tfor(int j = 0; j <= bounds.w; j++)\r\n\t{\r\n\t\tcurUV = float2(0.5,0.5) + float2( (float)i , (float)j );\r\n\t\tcurUV *= texelsize;\r\n\t\t\r\n\t\tcurUV.x = Quadrant.x == 0.0 ? curUV.x : 1.0 - curUV.x;\r\n\t\tcurUV.y = Quadrant.y == 0.0 ? curUV.y : 1.0 - curUV.y;\r\n\t\t\r\n\t\tcursample = Tex.SampleLevel(TexSampler, curUV, 0).r;\r\n\t\t\r\n\t\tif (cursample > Threshold)\r\n\t\t{\r\n\t\t\tfloat2 curvertex = float2((float)i , (float)j);\r\n\t\t\t\r\n\t\t\tfor(int k = 0; k < (bounds.w - bounds.z); k++)\r\n\t\t\t{\r\n\t\t\t\tfloat rise = (vertexA.y - k) - curvertex.y;\r\n\t\t\t\tfloat run  = curvertex.x - vertexA.x;\r\n\t\t\t\tfloat slope = rise/run;\r\n\t\t\t\t\r\n\t\t\t\tfloat xintercept = (bounds.w - bounds.z - k) / slope;\r\n\t\t\t\t\r\n\t\t\t\tfloat3 color = float3(1,0,0);\r\n\t\t\t\r\n\t\t\t\tif(slope > maxslopes[k])\r\n\t\t\t\t{\r\n\t\t\t\t\tcolor = float3(0,1,0);\r\n\t\t\t\t\tmaxslopes[k] = slope;\r\n\t\t\t\t\tmaxslopetris[k].xy = vertexA - float2(0,k);\r\n\t\t\t\t\tmaxslopetris[k].zw = bounds.xz + float2(xintercept,0);\r\n\t\t\t\t}\r\n\t\t\t\t\r\n\r\n\t\t\t\tif(1)\r\n\t\t\t\t{\r\n\t\t\t\t\tfloat2 curA = vertexA - float2(0,k);\r\n\t\t\t\t\tfloat2 curB = bounds.xz + float2(xintercept,0);\r\n\t\t\t\t\t\r\n\t\t\t\t\tcurA.x = Quadrant.x == 0.0 ? curA.x : 15.0 - curA.x;\r\n\t\t\t\t\tcurA.y = Quadrant.y == 0.0 ? curA.y : 15.0 - curA.y;\r\n\t\t\t\t\tcurB.x = Quadrant.x == 0.0 ? curB.x : 15.0 - curB.x;\r\n\t\t\t\t\tcurB.y = Quadrant.y == 0.0 ? curB.y : 15.0 - curB.y;\r\n\t\t\t\t\t\r\n\t\t\t\t\tcurA += saturate(Quadrant);\r\n\t\t\t\t\tcurB += saturate(Quadrant);\r\n\t\t\t\t\tdebug_lines += Helper.DrawLine(curA, curB, 0.05) * 0.5 * color;\r\n\t\t\t\t\t\r\n\t\t\t\t\tfloat2 localcorner = curvertex;\r\n\t\t\t\t\tlocalcorner.x = Quadrant.x == 0.0 ? localcorner.x : 15.0 - localcorner.x;\r\n\t\t\t\t\tlocalcorner.y = Quadrant.y == 0.0 ? localcorner.y : 15.0 - localcorner.y;\r\n\t\t\t\t\t\r\n\t\t\t\t\tdebug_lines += Helper.DrawCircle(localcorner + 0.5, 0.25, 0.1) * 0.25;\r\n\t\t\t\t\t\r\n\t\t\t\t\tif(slope == 0)\r\n\t\t\t\t\t{\r\n\t\t\t\t\t\tdebug_lines += Helper.DrawCircle(localcorner + 0.5, 0.1, 0.1) * 0.25 * float3(1,0,0);\r\n\t\t\t\t\t}\r\n\t\t\t\t}\r\n\t\t\t}\r\n\t\t\t\r\n\t\t\tbreak;\r\n\t\t}\r\n\t}\r\n\t\r\n\tif (cursample > Threshold)\r\n\t{\r\n\t\t//break;\r\n\t}\r\n}\r\n\r\nfloat maxarea = 0;\r\nint largestindex = 0;\r\n\r\n//now find the maxslopetriangle with the largest area\r\nfor(int i = 0; i < (bounds.w - bounds.z); i++)\r\n{\r\n\tfloat2 pointA = maxslopetris[i].xy;\r\n\tfloat2 pointB = maxslopetris[i].zw;\r\n\t\r\n\tfloat area = ((pointB.x - pointA.x) * (pointA.y - pointB.y)) / 2.0;\r\n\t\r\n\tif(area > maxarea)\r\n\t{\r\n\t\tmaxarea = area;\r\n\t\tlargestindex = i;\r\n\t}\r\n}\r\n\r\nfloat2 largestslice_A = maxslopetris[largestindex].xy;\r\nfloat2 largestslice_B = maxslopetris[largestindex].zw;\r\n\r\nlargestslice_A.x = Quadrant.x == 0.0 ? largestslice_A.x : 15.0 - largestslice_A.x;\r\nlargestslice_A.y = Quadrant.y == 0.0 ? largestslice_A.y : 15.0 - largestslice_A.y;\r\n\t\t\r\nlargestslice_B.x = Quadrant.x == 0.0 ? largestslice_B.x : 15.0 - largestslice_B.x;\r\nlargestslice_B.y = Quadrant.y == 0.0 ? largestslice_B.y : 15.0 - largestslice_B.y;\r\n\r\nlargestslice_A += saturate(Quadrant);\r\nlargestslice_B += saturate(Quadrant);\r\n\r\ndebug_lines += Helper.DrawLine(largestslice_A, largestslice_B , 0.25) * 0.5 * float3(0,0,2);\r\n\r\nfloat4 outdata = float4(largestslice_A.xy, largestslice_B.xy) / 16;\r\n\r\nOut_DebugLines = debug_lines.xyz;\r\n\r\nreturn outdata;"
   description=Cut Mesh Corners
   outputType=CMOT_Float4
   additionalOutputs [{outputName=Out_DebugLines, outputType=CMOT_Float1}]
   inputs [
    {inputName=Anchors, input {expression=&self:MaterialExpressionNamedRerouteUsage_7}}
    {inputName=Tex, input {expression=&self:MaterialExpressionReroute_0}}
    {inputName=Quadrant, input {expression=&self:MaterialExpressionNamedRerouteDeclaration_2}}
    {inputName=UV, input {expression=&self:MaterialExpressionTextureCoordinate_4}}
    {inputName=Threshold, input {expression=&self:MaterialExpressionScalarParameter_11}}
   ]
  }
  NamedRerouteDeclaration(None=@n9.return) {name=Vertices, nodeColor=<rgba 1 0.3984 0 1>}
  n10 = TextureCoordinate()
  n11 = MaterialFunctionCall!(
   Vector4=NamedRerouteUsage()
   MaximumNumberOfDigits
   UVs=Frac(None=Multiply(A=@n10) {constB=2})
   DebugTextLocation RG_UpperRight BA_LowerLeft
   Component Spacing
  ) {materialFunction=/Engine/Functions/Engine_MaterialFunctions02/Utility/DebugFloat4Values}
  n12 = ScalarParameter() {defaultValue=1, parameterName=DebugGridSoften}
  n13 = Multiply(A=TextureCoordinate()) {constB=16}
  n14 = ScalarParameter() {defaultValue=0.001, parameterName=DebugGridThickness}
  n15 = Multiply(A=OneMinus(None=MaterialFunctionCall!(InvSoften=@n12, UV=@n13, Tweak=@n14) {materialFunction=/Engine/EditorMaterials/AntialiasedGridFloor}.Result)) {constB=0.2}
  NamedRerouteDeclaration(None=Add(
    A=ComponentMask(None=@n11.ColorCodedOutput) {b=true, g=true, r=true}
    B=NamedRerouteDeclaration(None=@n15) {name=Debug Grid, nodeColor=<rgba 0.9453 1 0 1>}
   )) {name=Values, nodeColor=<rgba 1 0.0234 0 1>}
  MaterialExpressionNamedRerouteDeclaration_6 = NamedRerouteDeclaration(None=TextureObjectParameter() {parameterName=RT, samplerType=SAMPLERTYPE_LinearColor, texture=/ImpostorBaker/TestContent/RTs/RT_Test_01}) {name=Texture Object, nodeColor=<rgba 0 1 0.6719 1>}
  MaterialExpressionNamedRerouteUsage_6 = NamedRerouteUsage()
  MaterialExpressionScalarParameter_10 = ScalarParameter() {defaultValue=0.25, parameterName=Threshold}
  n16 = Custom!(
   Tex=@MaterialExpressionNamedRerouteDeclaration_6
   Quadrant=@MaterialExpressionNamedRerouteUsage_6
   Threshold=@MaterialExpressionScalarParameter_10
  ) {
   code="\r\n\r\nfloat texelsize = 1.0 / 16.0;\r\n\r\nfloat cursample = 0;\r\nfloat2 corner_x_trace = float2(0,0);\r\nfloat2 corner_y_trace = float2(0,0);\r\nfloat2 curUV = float2(0,0);\r\n\r\n//first trace X, j is x and i is y\r\nfor(int i = 0; i < 7; i++)\r\n{\r\n\tfor(int j = 0; j < 15; j++)\r\n\t{\r\n\t\tcurUV = float2(0.5,0.5) + float2( (float)j , (float)i );\r\n\t\tcurUV *= texelsize;\r\n\t\t\r\n\t\tcurUV.x = Quadrant.x == 0.0 ? curUV.x : 1.0 - curUV.x;\r\n\t\tcurUV.y = Quadrant.y == 0.0 ? curUV.y : 1.0 - curUV.y;\r\n\t\t\r\n\t\tcursample = Tex.SampleLevel(TexSampler, curUV, 0).r;\r\n\t\t\r\n\t\tif (cursample > Threshold)\r\n\t\t{\r\n\t\t\tcorner_x_trace = float2(j, i);\r\n\t\t\t\r\n\t\t\tbreak;\r\n\t\t}\r\n\t}\r\n\t\r\n\tif (cursample > Threshold)\r\n\t{\r\n\t\tbreak;\r\n\t}\r\n}\r\n\r\n//next trace Y\r\nfor(int k = 0; k < 7; k++)\r\n{\r\n\tfor(int l = 0; l < 15; l++)\r\n\t{\r\n\t\tcurUV = float2(0.5,0.5) + float2( (float)k , (float)l );\r\n\t\tcurUV *= texelsize;\r\n\t\t\r\n\t\tcurUV.x = Quadrant.x == 0.0 ? curUV.x : 1.0 - curUV.x;\r\n\t\tcurUV.y = Quadrant.y == 0.0 ? curUV.y : 1.0 - curUV.y;\r\n\t\t\r\n\t\tcursample = Tex.SampleLevel(TexSampler, curUV, 0).r;\r\n\t\t\r\n\t\tif (cursample > Threshold)\r\n\t\t{\r\n\t\t\tcorner_y_trace = float2(k, l);\r\n\t\t\t\r\n\t\t\tbreak;\r\n\t\t}\r\n\t}\r\n\t\r\n\tif (cursample > Threshold)\r\n\t{\r\n\t\tbreak;\r\n\t}\r\n}\r\n\r\nreturn float4(corner_x_trace, corner_y_trace);"
   description=Find Bounds and Anchors
   outputType=CMOT_Float4
   inputs [
    {inputName=Tex, input {expression=&self:MaterialExpressionNamedRerouteDeclaration_6}}
    {inputName=Quadrant, input {expression=&self:MaterialExpressionNamedRerouteUsage_6}}
    {inputName=Threshold, input {expression=&self:MaterialExpressionScalarParameter_10}}
   ]
  }
  MaterialExpressionNamedRerouteDeclaration_5 = NamedRerouteDeclaration(None=@n16) {name=Anchor Points, nodeColor=<rgba 0.4766 0 1 1>}
  MaterialExpressionNamedRerouteUsage_3 = NamedRerouteUsage()
  MaterialExpressionTextureCoordinate_5 = TextureCoordinate()
  MaterialExpressionConstant3Vector_0 = Constant3Vector() {constant=<rgba 0 0.5 0.5 0>}
  MaterialExpressionConstant3Vector_1 = Constant3Vector() {constant=<rgba 0.5 0 0.5 0>}
  n17 = Custom!(
   Verts=@MaterialExpressionNamedRerouteDeclaration_5
   Quadrant=@MaterialExpressionNamedRerouteUsage_3
   UV=@MaterialExpressionTextureCoordinate_5
   ColorA=@MaterialExpressionConstant3Vector_0
   ColorB=@MaterialExpressionConstant3Vector_1
  ) {
   code="struct ShapeDrawer\r\n{\r\n\r\n\tfloat2 P;\r\n\t\r\n\tfloat DrawLine(float2 inA, float2 inB, float inWidth)\r\n\t{\r\n\t\tfloat2 UV = P - inA;\r\n\t\tfloat2 squashvec = inB - inA;\r\n\t\t\r\n\t\t\r\n\t\t\r\n\t\tfloat2 distalongline =  saturate( dot(UV, squashvec) / dot(squashvec, squashvec)) * squashvec;\r\n\t\t\r\n\t\tfloat2 squashUV = UV - distalongline;\r\n\t\t\r\n\t\tfloat thisline = length(squashUV);\r\n\t\tthisline = saturate(-((thisline - inWidth)/ inWidth));\r\n\t\t\r\n\t\treturn thisline;\r\n\t}\r\n\t\r\n\tfloat DrawCircle(float2 inC, float inRadius, float inWidth)\r\n\t{\r\n\t\tfloat2 UV = P - inC;\r\n\t\tinWidth/=2;\r\n\t\r\n\t\tfloat dist = abs( length(UV) - inRadius);\r\n\t\t\r\n\t\tdist = saturate(-((dist - inWidth)/ inWidth));\r\n\t\treturn dist;\r\n\t}\r\n\t\r\n\tfloat DrawBox(float2 inC, float2 inExtents, float inWidth)\r\n\t{\r\n\t\tfloat2 pos = abs(P - inC) / inExtents;\r\n\t\tfloat dist = abs(max(pos.x-1, pos.y-1));\r\n\t\tdist = dist - inWidth*2;\r\n\t\tdist = saturate(-((dist - inWidth)/ inWidth));\r\n\t\t\r\n\t\treturn dist;\r\n\t}\r\n\t\r\n\tfloat2 GetFocalDirection(float2 inFocalPoint, float2 inRayAtFocalDepth, float2 inOffset)\r\n\t{\r\n\t\tfloat2 gd = normalize( inRayAtFocalDepth - inOffset);\r\n\t\treturn normalize(inRayAtFocalDepth + gd);\r\n\t}\r\n\t\r\n\tfloat2 IntersectCircle( float2 inC, float inR, float2 inRO, float2 inRD)\r\n\t{\r\n\t\tinRO = inRO - inC;\r\n\t\tfloat B = dot( inRD, inRO);\r\n\t\tfloat C = dot( inRO, inRO) - (inR*inR);\r\n\t\tfloat D = sqrt(B*B - C);\r\n\t\tfloat t0 = -B - D;\r\n\t\tfloat t1 = -B + D;\r\n\t\t\r\n\t\tt0 = max(0, t0);\r\n\t\t\t\t\r\n\t\treturn float2(t0, t1);\r\n\t}\r\n\t\r\n\tfloat2 IntersectBox(float2 inC, float2 inExtents, float2 inRO, float2 inRD)\r\n\t{\r\n\t\tinRO -= inC;\r\n\t\t\r\n\t\tfloat2 invraydir = 1.0f / inRD;\r\n\t\tfloat2 firstintersections = (-inExtents - inRO) * invraydir;\r\n\t\tfloat2 secondintersections = (inExtents - inRO) * invraydir;\r\n\t\tfloat2 closest = min(firstintersections, secondintersections);\r\n\t\tfloat2 furthest = max(firstintersections, secondintersections);\r\n\t\t\r\n\t\tfloat t0 = max(closest.x, closest.y);\r\n\t\tfloat t1 = min(furthest.x, furthest.y);\r\n\t\tt0 = max(0, t0);\r\n\t\t\r\n\t\treturn float2(t0, t1);\r\n\t}\r\n\t\r\n};\r\n\r\nShapeDrawer Helper;\r\nHelper.P = UV * 16;\r\n\r\nfloat2 VertexA = Verts.xy;\r\nfloat2 VertexB = Verts.zw;\r\n\r\nVertexA.x = Quadrant.x == 0.0 ? VertexA.x : 15.0 - VertexA.x;\r\nVertexA.y = Quadrant.y == 0.0 ? VertexA.y : 15.0 - VertexA.y;\r\n\t\t\r\nVertexB.x = Quadrant.x == 0.0 ? VertexB.x : 15.0 - VertexB.x;\r\nVertexB.y = Quadrant.y == 0.0 ? VertexB.y : 15.0 - VertexB.y;\r\n\r\nVertexA += saturate(Quadrant);\r\nVertexB += saturate(Quadrant);\r\n\r\nfloat3 accum = 0;\r\n\r\naccum += Helper.DrawCircle(VertexA, 0.5, 0.25) * ColorA;\r\naccum += Helper.DrawCircle(VertexB, 0.5, 0.25) * ColorB;\r\n\r\nreturn accum;\r\n"
   description=Draw Anchors
   inputs [
    {inputName=Verts, input {expression=&self:MaterialExpressionNamedRerouteDeclaration_5}}
    {inputName=Quadrant, input {expression=&self:MaterialExpressionNamedRerouteUsage_3}}
    {inputName=UV, input {expression=&self:MaterialExpressionTextureCoordinate_5}}
    {
     inputName=ColorA
     input {expression=&self:MaterialExpressionConstant3Vector_0, mask=1, maskB=1, maskG=1, maskR=1}
    }
    {
     inputName=ColorB
     input {expression=&self:MaterialExpressionConstant3Vector_1, mask=1, maskB=1, maskG=1, maskR=1}
    }
   ]
  }
  NamedRerouteDeclaration(None=@n17) {name=Anchor Points Debug, nodeColor=<rgba 1 0 0.5625 1>}
  n18 = Floor(None=Multiply(A=TextureCoordinate()) {constB=16})
  n19 = TextureSampleParameter2D!(UVs=Divide(A=Add(A=@n18) {constB=0.5}) {constB=16}, Level, Apply View MipBias) {mipValueMode=TMVM_MipLevel, parameterName=RT, samplerType=SAMPLERTYPE_LinearColor, texture=&RT_Test_01}
  NamedRerouteDeclaration(None=Multiply(A=@n19.R) {constB=0.1}) {name=Texture Quantized, nodeColor=<rgba 0 0.0156 1 1>}
  NamedRerouteDeclaration(None=@n9.Out_DebugLines) {name=Debug Lines, nodeColor=<rgba 0.9688 0 1 1>}
 ]
}
