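# Unlit material for Impostor Baker generation that computes planar UVs in a view frame around a component origin.
# World position minus ComponentOrigin is transformed into the basis given by the X, Y and Z vector parameters through InverseTransformMatrix.
# The result is divided by scalar R, remapped with ConstantBiasScale and masked to red and green, giving a 2D coordinate.
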Material /ImpostorBaker/Materials/Generation/M_ViewFrame_UV
graph {
 properties {pixelDepthOffsetMode=PDOM_Legacy, shadingModel=MSM_Unlit, usedShadingModels=Unlit}
 nodes [
  n1 = MaterialFunctionCall!(
   VectorToTransform=Subtract(A=WorldPosition().XYZ, B=VectorParameter() {parameterName=ComponentOrigin}.RGB)
   BasisX=VectorParameter() {parameterName=X, defaultValue=<rgba 1 0 0 0>}.RGB
   BasisY=VectorParameter() {parameterName=Y, defaultValue=<rgba 0 1 0 0>}.RGB
   BasisZ=VectorParameter() {parameterName=Z, defaultValue=<rgba 0 0 1 0>}.RGB
  ) {materialFunction=/Engine/Functions/Engine_MaterialFunctions02/Math/InverseTransformMatrix}
  n2 = ScalarParameter() {defaultValue=256, parameterName=R}
  ComponentMask(None=ConstantBiasScale(None=Divide(A=@n1.Transformed Vector 3, B=@n2))) {g=true, r=true}
 ]
}
